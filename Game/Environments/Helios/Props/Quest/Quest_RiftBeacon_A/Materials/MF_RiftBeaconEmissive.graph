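# Material function that adds animated emissive panel glow to the Rift Beacon quest prop's material attributes.
# Base color channels and local position drive time-based sine waves, which combine with a tiled cell texture into stepped cell patterns.
# Named reroutes carry camera-distance fade, an emissive area mask, and a fade-in with climax timed by custom primitive data.
# The tinted result is compensated with inverse eye adaptation; a debug graph plot remains.
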MaterialFunction /Game/Environments/Helios/Props/Quest/Quest_RiftBeacon_A/Materials/MF_RiftBeaconEmissive
graph {nodes [
  Add()
  ConstantBiasScale(None=ComponentMask(None=TextureCoordinate()) {r=true}) {bias=-0.5, scale=4.4}
  n1 = TextureSample() {texture=/Game/Environments/Helios/Props/Quest/Quest_RiftBeacon_A/Textures/T_Quest_RiftBeacon_Panels_A_A_D}
  n2 = GetMaterialAttributes(None=FunctionInput(Preview=MakeMaterialAttributes!(
     BaseColor=@n1.RGB
     Metallic
     Specular
     Roughness
     Anisotropy
     EmissiveColor
     Opacity
     OpacityMask
     Normal
     Tangent
     WorldPositionOffset
     SubsurfaceColor
     ClearCoat
     ClearCoatRoughness
     AmbientOcclusion
     Refraction
     CustomizedUVs_0
     CustomizedUVs_1
     CustomizedUVs_2
     CustomizedUVs_3
     CustomizedUVs_4
     CustomizedUVs_5
     CustomizedUVs_6
     CustomizedUVs_7
     PixelDepthOffset
     ShadingModel
     Displacement
    )) {inputName=MA, inputType=FunctionInput_MaterialAttributes}) {attributeGetTypes [#1, #2]}
  n3 = NamedRerouteUsage()
  n4 = Saturate(None=ConstantBiasScale(None=ComponentMask(None=@n2.BaseColor) {b=true}) {bias=-0.076, scale=-10})
  n5 = NamedRerouteUsage()
  n6 = Multiply(A=@n5, B=@n5)
  n7 = ComponentMask(None=@n2.BaseColor) {g=true}
  n8 = OneMinus(None=Saturate(None=ConstantBiasScale(None=@n7) {bias=0, scale=1}))
  n9 = VertexInterpolator(VS=LocalPosition() {localOrigin=InstancePreSkinning}.XYZ)
  n10 = Constant3Vector() {constant=<rgba 0 0 -160 0>}
  n11 = MaterialFunctionCall(Color=Abs(None=Add(A=@n9.PS, B=@n10))) {materialFunction=&SplitComponents}
  n12 = Saturate(None=Divide(A=Add(A=@n11.R, B=@n11.R), B=Constant() {desc=130, r=130}) {constB=150})
  n13 = Add(
   A=ScalarParameter() {defaultValue=1.5, group=RiftBeacon_Emissive, parameterName=Anim_Wave Frequency}
   B=Multiply(A=Multiply(A=@n12, B=@n12), B=Constant() {r=-1}) {constB=0.5}
  )
  n14 = Multiply(A=@n8, B=@n13) {constB=1.5}
  n15 = Time()
  n16 = NamedRerouteUsage()
  n17 = Divide(
   A=Add(A=NamedRerouteDeclaration(None=@n15) {name=In_Time, nodeColor=<rgba 0.8516 1 0 1>}, B=Multiply(A=@n16) {constB=-2})
   B=Constant() {desc=4, r=4}
  ) {constB=-3}
  n18 = Add(A=@n17) {constB=10}
  n19 = Multiply(A=@n8, B=Multiply(A=@n13) {constB=3}) {constB=5}
  n20 = Sine(None=Add(A=Add(A=@n14, B=@n17), B=Add(A=@n19, B=@n18)))
  n21 = ScalarParameter() {defaultValue=0.5, group=RiftBeacon_Emissive, parameterName=Cell_Bias}
  n22 = ScalarParameter() {defaultValue=0.45, group=RiftBeacon_Emissive, parameterName=Cell_Range Contrast}
  n23 = OneMinus(None=Multiply(A=Add(A=@n20, B=@n21), B=@n22) {constB=0.23})
  n24 = TextureSample(UVs=Multiply(A=TextureCoordinate(), B=ScalarParameter() {defaultValue=150, group=RiftBeacon_Emissive, parameterName=Cell_Tile}) {constB=100}) {
   samplerType=SAMPLERTYPE_LinearGrayscale
   texture=/Game/Environments/Helios/Props/Quest/Quest_RiftBeacon_A/Textures/T_Quest_RiftBeacon_Panels_A_A_D_Cells
  }
  n25 = Multiply(A=Add(A=OneMinus(None=@n24.R), B=Constant() {r=0.2}), B=Constant() {r=1.9}) {constB=0.2}
  n26 = Subtract(A=@n23, B=@n25)
  n27 = CameraPositionWS()
  n28 = ObjectPositionWS()
  n29 = Saturate(None=Divide(A=Distance(A=@n27, B=@n28), B=Constant() {r=1500}))
  n30 = Multiply(
   A=Step(Y=Constant() {r=0.3}, X=Add(A=@n23, B=@n26)) {constY=0.1}
   B=OneMinus(None=NamedRerouteDeclaration(None=@n29) {name=DistanceFade, nodeColor=<rgba 1 0.3047 0 1>})
  )
  n31 = Multiply(A=OneMinus(None=NamedRerouteUsage())) {constB=0.5}
  n32 = Multiply(A=Add(A=@n23, B=Subtract(A=@n30, B=@n31)), B=Constant() {r=0.02}) {constB=0.02}
  n33 = Multiply(A=Divide(A=@n22) {constB=10}, B=OneMinus(None=NamedRerouteUsage()))
  n34 = Multiply(A=Max(A=Multiply(A=@n4, B=@n6), B=Subtract(A=@n32, B=@n33))) {constB=10}
  n35 = EyeAdaptationInverse()
  n36 = Multiply(
   A=Multiply(
    A=Multiply(A=@n34, B=@n35.EyeAdaptationInverse)
    B=VectorParameter() {group=RiftBeacon_Emissive, parameterName=Panel_EmissiveColor, defaultValue=<rgba 0 3.4033 5 1>}.RGB
   )
   B=NamedRerouteUsage()
  )
  FunctionOutput(None=SetMaterialAttributes!(MaterialAttributes=@n2.MaterialAttributes, EmissiveColor=Add(A=@n3, B=@n36)) {attributeSetTypes [#2]})
  LinearInterpolate(
   A=ScalarParameter() {defaultValue=10, group=RiftBeacon_Emissive, parameterName=Emissive_Distance Min}
   B=ScalarParameter() {defaultValue=10, group=RiftBeacon_Emissive, parameterName=Emissive_Distance Max}
   Alpha=NamedRerouteUsage()
  )
  n37 = NamedRerouteUsage()
  n38 = ScalarParameter() {bUseCustomPrimitiveData=true, defaultValue=1, group=RiftBeacon_Emissive, parameterName=FadeInTime}
  n39 = Divide(A=Abs(None=Subtract(A=@n37, B=@n38)), B=ScalarParameter() {defaultValue=0.75, group=RiftBeacon_Emissive, parameterName=FadeIn_Duration})
  n40 = NamedRerouteDeclaration(None=Multiply(A=Saturate(None=@n39), B=Saturate(None=@n38))) {name=FadeIn, nodeColor=<rgba 0.3594 0 0.0075 1>}
  n41 = ScalarParameter() {defaultValue=0.2, group=RiftBeacon_Emissive, parameterName=Anim_Intro Climax}
  n42 = OneMinus(None=Multiply(A=Saturate(None=@n40), B=Add(A=@n41)))
  n43 = OneMinus(None=@n41)
  n44 = Multiply(A=OneMinus(None=Abs(None=@n42)), B=Add(A=Divide(A=@n41, B=@n43)))
  Multiply(A=MaterialFunctionCall!(
    X range Y range=Constant4Vector() {constant=<rgba -2 2 -2 2>}.RGBA
    Gradient Height
    ShowGrid?
    Thickness=Constant() {r=0.5}
    Color
    "f(x)"=@n44
   ) {materialFunction=/Engine/Functions/Engine_MaterialFunctions01/Debug/PlotFunctionOnGraph}."f(x) = y")
  Multiply(A=@n25, B=NamedRerouteUsage())
  n45 = Step(Y=ComponentMask(None=@n2.BaseColor) {r=true}) {constX=0.01}
  NamedRerouteDeclaration(None=Multiply(A=Saturate(None=@n45), B=Step(Y=@n11.B) {constX=90, constY=1})) {name=M_EmissiveArea, nodeColor=<rgba 0 1 1 1>}
  NamedRerouteDeclaration(None=@n44) {name=FadeIn_Climax, nodeColor=<rgba 0.6875 0 1 1>}
  NamedRerouteDeclaration(None=@n2.EmissiveColor) {name=IN_Emissive, nodeColor=<rgba 0 1 0.625 1>}
  NamedRerouteUsage()
 ]}
